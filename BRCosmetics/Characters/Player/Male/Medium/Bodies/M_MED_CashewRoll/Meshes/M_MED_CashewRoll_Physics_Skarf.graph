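# Physics asset for the scarf on the male medium CashewRoll cosmetic body, with capsule bodies for the scarf bones, upper spine and clavicles.
# A three-segment tail and two two-segment bows each hang from a small base body constrained to spine_05.
# Segment joints use tight swing and twist limits, while torso and base joints leave rotation free.
# Left and right bows differ slightly in capsule shapes and limits.
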PhysicsAsset /BRCosmetics/Characters/Player/Male/Medium/Bodies/M_MED_CashewRoll/Meshes/M_MED_CashewRoll_Physics_Skarf
bodies {
 clavicle_l [{shapeType=Capsule, length=9.3913, radius=9.4502, shapeName=clavicle_l_capsule, center=<7.719 0 0>, rotation=<rot 90 0 0>}]
 clavicle_r [{shapeType=Capsule, length=9.3913, radius=9.4502, shapeName=clavicle_l_capsule, center=<-7.719 0 0>, rotation=<rot 90 0 180>}]
 dyn_skarf_01 [{shapeType=Capsule, length=12.6238, radius=1.8934, shapeName=dyn_skarf_01_capsule, center=<5.0501 0 0>}]
 dyn_skarf_02 [{shapeType=Capsule, length=9.4615, radius=1.2842, shapeName=dyn_skarf_02_capsule, center=<6.9012 0 0>}]
 dyn_skarf_03 [{shapeType=Capsule, length=1.5285, radius=1.3011, shapeName=dyn_skarf_03_capsule, center=<4.2337 0.1705 0>}]
 dyn_skarf_BASE [{shapeType=Capsule, length=0.505, radius=0.505, shapeName=dyn_skarf_BASE_capsule}]
 dyn_skarf_bow_01_l [{
   shapeType=Capsule
   length=7.9665
   radius=1.6238
   shapeName=dyn_skarf_bow_01_l_capsule
   center=<7.7699 0 -0.7621>
   rotation=<rot -9.3911 20.2838 -3.452>
  }]
 dyn_skarf_bow_01_r [{
   shapeType=Capsule
   length=7.9665
   radius=1.6238
   shapeName=dyn_skarf_bow_01_l_capsule
   center=<9.4371 0.0918 0.2513>
   rotation=<rot 0.5911 19.6847 176.5944>
  }]
 dyn_skarf_bow_02_l [{
   shapeType=Capsule
   length=9.4093
   radius=1.4202
   shapeName=dyn_skarf_bow_02_r_capsule
   center=<6.5292 0 -1.225>
   rotation=<rot 90 -179.3669 0.6324>
  }]
 dyn_skarf_bow_02_r [{shapeType=Capsule, length=6.0064, radius=2, shapeName=dyn_skarf_bow_02_r_capsule, center=<5.6945 0 0.4376>, rotation=<rot 85 0.003 0.0028>}]
 dyn_skarf_bow_BASE_l [{shapeType=Capsule, length=1, radius=1, shapeName=dyn_skarf_bow_BASE_l_capsule}]
 dyn_skarf_bow_BASE_r [{shapeType=Capsule, length=1, radius=1, shapeName=dyn_skarf_bow_BASE_r_capsule}]
 spine_03 [{shapeType=Capsule, length=12.9497, radius=12.3088, shapeName=spine_03_capsule, rotation=<rot 0 -90 0>}]
 spine_04 [{shapeType=Capsule, length=7.5136, radius=11.4158, shapeName=spine_04_capsule, rotation=<rot 0 -90 0>}]
 spine_05 [{shapeType=Capsule, length=12.7918, radius=10.6215, shapeName=spine_05_capsule, center=<0 -1.3625 0>, rotation=<rot 0 -90 0>}]
}
constraints [
 {
  bone1Name=spine_04
  bone2Name=spine_03
  swing1LimitDegrees=45
  swing1Motion=Free
  swing2LimitDegrees=45
  swing2Motion=Free
  twistLimitDegrees=45
  twistMotion=Free
 }
 {
  bone1Name=spine_05
  bone2Name=spine_04
  swing1LimitDegrees=45
  swing1Motion=Free
  swing2LimitDegrees=45
  swing2Motion=Free
  twistLimitDegrees=45
  twistMotion=Free
 }
 {
  bone1Name=dyn_skarf_BASE
  bone2Name=spine_05
  swing1LimitDegrees=45
  swing1Motion=Free
  swing2LimitDegrees=45
  swing2Motion=Free
  twistLimitDegrees=45
  twistMotion=Free
 }
 {bone1Name=dyn_skarf_01, bone2Name=dyn_skarf_BASE, swing1LimitDegrees=20, swing2LimitDegrees=12, twistLimitDegrees=8}
 {bone1Name=dyn_skarf_02, bone2Name=dyn_skarf_01, swing1LimitDegrees=20, swing2LimitDegrees=12, twistLimitDegrees=8}
 {bone1Name=dyn_skarf_03, bone2Name=dyn_skarf_02, swing1LimitDegrees=20, swing2LimitDegrees=12, twistLimitDegrees=8}
 {
  bone1Name=dyn_skarf_bow_BASE_l
  bone2Name=spine_05
  swing1LimitDegrees=45
  swing1Motion=Free
  swing2LimitDegrees=45
  swing2Motion=Free
  twistLimitDegrees=45
  twistMotion=Free
 }
 {bone1Name=dyn_skarf_bow_01_l, bone2Name=dyn_skarf_bow_BASE_l, swing1LimitDegrees=4, swing2LimitDegrees=13, twistLimitDegrees=6}
 {
  bone1Name=dyn_skarf_bow_BASE_r
  bone2Name=spine_05
  swing1LimitDegrees=45
  swing1Motion=Free
  swing2LimitDegrees=45
  swing2Motion=Free
  twistLimitDegrees=45
  twistMotion=Free
 }
 {bone1Name=dyn_skarf_bow_02_r, bone2Name=dyn_skarf_bow_01_r, swing1LimitDegrees=8, swing2LimitDegrees=12, twistLimitDegrees=6}
 {
  bone1Name=clavicle_l
  bone2Name=spine_05
  swing1LimitDegrees=45
  swing1Motion=Free
  swing2LimitDegrees=45
  swing2Motion=Free
  twistLimitDegrees=45
  twistMotion=Free
 }
 {
  bone1Name=clavicle_r
  bone2Name=spine_05
  swing1LimitDegrees=45
  swing1Motion=Free
  swing2LimitDegrees=45
  swing2Motion=Free
  twistLimitDegrees=45
  twistMotion=Free
 }
 {bone1Name=dyn_skarf_bow_01_r, bone2Name=dyn_skarf_bow_BASE_r, swing1LimitDegrees=3, swing2LimitDegrees=13, twistLimitDegrees=6}
 {bone1Name=dyn_skarf_bow_02_l, bone2Name=dyn_skarf_bow_01_l, swing1LimitDegrees=10, swing2LimitDegrees=12, twistLimitDegrees=6}
]
