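# Post-process style AnimBP for the male CashewRoll body cosmetic, adding secondary motion to the parent's copied pose.
# A deform Control Rig feeds AnimDynamics for shoelaces and belt bags, then RigidBody for the scarf.
# In front end, scarf bows get additive rotations and DefaultSlot overrides belt gun/bag bones by the Frontend_Idle_Bypass curve.
# Updates also compute gravity override and drive material WPO cloth from calf motion, with wind while skydiving.
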AnimBlueprint /BRCosmetics/Characters/Player/Male/Medium/Bodies/M_MED_CashewRoll/Meshes/M_MED_CashewRoll_AnimBP
target_skeleton=&/M_MED_CashewRoll_Skeleton
graphs [
 AnimGraph {nodes [
   n1 = Layered blend per bone(
    BasePose=Use cached pose 'FECustomIdleCache'()
    BlendPoses_0=Slot 'DefaultSlot'()
    BlendWeights_0=Get CustomIdleIsPlayingFloat() -> (CustomIdleIsPlayingFloat:float)
   ) {layerSetup [{branchFilters [{boneName=dyn_belt_gun_BASE}, {boneName=dyn_belt_bag_back_BASE}, {boneName=dyn_belt_bag_front_BASE}]}]}
   Output Pose(Result=@n1) {layerGroup=None}
   n2 = Control Rig(Source=Copy Pose From Mesh() {bUseAttachedParent=true}) {
    alphaBoolBlend {blendOption=Linear}
    controlRigAssetReference {blueprintRigClass=&/M_MED_CashewRoll_CtrlRig_Deform.M_MED_CashewRoll_CtrlRig_Deform_C}
   }
   n3 = Anim Dynamics - Bone: dyn_shoe_strings_bot_l_l!(ComponentPose=Local To Component(LocalPose=@n2), Alpha=1, GravityScale:float=1, GravityOverride:Vector="0,0,-490", ExternalForce="0,0,0") -> (Pose) {
    angularDampingOverride=0.9999
    angularSpringConstant=100
    bAngularSpring=true
    bGravityOverrideInSimSpace=true
    bOverrideAngularDamping=true
    simulationSpace=RootRelative
    alphaBoolBlend {blendOption=Linear}
    boundBone {boneName=dyn_shoe_strings_bot_l_l}
    componentAppliedLinearAccClamp=<5000 5000 5000>
    componentLinearAccScale=<0.6 0.6 0.6>
    componentLinearVelScale=<0.6 0.6 0.6>
    gravityOverride {z=-980}
    physicsBodyDefinitions [{
      sphereCollisionRadius=0
      boundBone {boneName=dyn_shoe_strings_bot_l_l}
      boxExtents=<13.58 7.14 6.86>
      constraintSetup {angularLimitsMax=<7 5 0>, angularLimitsMin=<-10 -12 -70>, angularTarget=<1 0 0>}
      localJointOffset=<3.528 0 0>
     }]
    relativeSpaceBone {boneName=pelvis}
   }
   n4 = Anim Dynamics - Bone: dyn_shoe_strings_bot_r_l!(ComponentPose=@n3, Alpha=1, GravityScale:float=1, GravityOverride:Vector="0,0,-490", ExternalForce="0,0,0") -> (Pose) {
    angularDampingOverride=0.9999
    angularSpringConstant=100
    bAngularSpring=true
    bGravityOverrideInSimSpace=true
    bOverrideAngularDamping=true
    simulationSpace=RootRelative
    alphaBoolBlend {blendOption=Linear}
    boundBone {boneName=dyn_shoe_strings_bot_r_l}
    componentAppliedLinearAccClamp=<5000 5000 5000>
    componentLinearAccScale=<0.6 0.6 0.6>
    componentLinearVelScale=<0.6 0.6 0.6>
    gravityOverride {z=-980}
    physicsBodyDefinitions [{
      sphereCollisionRadius=0
      boundBone {boneName=dyn_shoe_strings_bot_r_l}
      boxExtents=<14.28 6.86 7.14>
      constraintSetup {angularLimitsMax=<7 5 0>, angularLimitsMin=<-10 -12 -70>, angularTarget=<1 0 0>}
      localJointOffset=<3.8752 0 0>
     }]
    relativeSpaceBone {boneName=pelvis}
   }
   n5 = Anim Dynamics - Bone: dyn_shoe_strings_top_l_l!(ComponentPose=@n4, Alpha=1, GravityScale:float=1, GravityOverride:Vector="0,0,-490", ExternalForce="0,0,0") -> (Pose) {
    angularDampingOverride=0.9999
    angularSpringConstant=100
    bAngularSpring=true
    bGravityOverrideInSimSpace=true
    bOverrideAngularDamping=true
    simulationSpace=RootRelative
    alphaBoolBlend {blendOption=Linear}
    boundBone {boneName=dyn_shoe_strings_top_l_l}
    componentAppliedLinearAccClamp=<5000 5000 5000>
    componentLinearAccScale=<0.6 0.6 0.6>
    componentLinearVelScale=<0.6 0.6 0.6>
    gravityOverride {z=-980}
    physicsBodyDefinitions [{
      sphereCollisionRadius=0
      boundBone {boneName=dyn_shoe_strings_top_l_l}
      boxExtents=<14 7 7>
      constraintSetup {angularLimitsMax=<7 5 0>, angularLimitsMin=<-10 -12 -70>, angularTarget=<1 0 0>}
      localJointOffset=<3.92 0 0>
     }]
    relativeSpaceBone {boneName=pelvis}
   }
   n6 = Anim Dynamics - Bone: dyn_shoe_strings_top_r_l!(ComponentPose=@n5, Alpha=1, GravityScale:float=1, GravityOverride:Vector="0,0,-490", ExternalForce="0,0,0") -> (Pose) {
    angularDampingOverride=0.9999
    angularSpringConstant=100
    bAngularSpring=true
    bGravityOverrideInSimSpace=true
    bOverrideAngularDamping=true
    simulationSpace=RootRelative
    alphaBoolBlend {blendOption=Linear}
    boundBone {boneName=dyn_shoe_strings_top_r_l}
    componentAppliedLinearAccClamp=<5000 5000 5000>
    componentLinearAccScale=<0.6 0.6 0.6>
    componentLinearVelScale=<0.6 0.6 0.6>
    gravityOverride {z=-980}
    physicsBodyDefinitions [{
      sphereCollisionRadius=0
      boundBone {boneName=dyn_shoe_strings_top_r_l}
      boxExtents=<14.28 6.86 6.72>
      constraintSetup {angularLimitsMax=<7 5 0>, angularLimitsMin=<-10 -12 -70>, angularTarget=<1 0 0>}
      localJointOffset=<4.116 0 0>
     }]
    relativeSpaceBone {boneName=pelvis}
   }
   n7 = Anim Dynamics - Bone: dyn_shoe_strings_bot_l_r!(ComponentPose=@n6, Alpha=1, GravityScale:float=1, GravityOverride:Vector="0,0,-980", ExternalForce="0,0,0") -> (Pose) {
    angularDampingOverride=0.9999
    angularSpringConstant=100
    bAngularSpring=true
    bGravityOverrideInSimSpace=true
    bOverrideAngularDamping=true
    simulationSpace=RootRelative
    alphaBoolBlend {blendOption=Linear}
    boundBone {boneName=dyn_shoe_strings_bot_l_r}
    componentAppliedLinearAccClamp=<5000 5000 5000>
    componentLinearAccScale=<0.6 0.6 0.6>
    componentLinearVelScale=<0.6 0.6 0.6>
    gravityOverride {z=-980}
    physicsBodyDefinitions [{
      sphereCollisionRadius=0
      boundBone {boneName=dyn_shoe_strings_bot_l_r}
      boxExtents=<13.58 7.14 6.86>
      constraintSetup {angularLimitsMax=<7 5 0>, angularLimitsMin=<-10 -12 -70>, angularTarget=<0 0 0>}
      localJointOffset=<3.528 0 0>
     }]
    relativeSpaceBone {boneName=foot_l}
   }
   n8 = Anim Dynamics - Bone: dyn_shoe_strings_bot_r_r!(ComponentPose=@n7, Alpha=1, GravityScale:float=1, GravityOverride:Vector="0,0,-980", ExternalForce="0,0,0") -> (Pose) {
    angularDampingOverride=0.9999
    angularSpringConstant=100
    bAngularSpring=true
    bGravityOverrideInSimSpace=true
    bOverrideAngularDamping=true
    simulationSpace=RootRelative
    alphaBoolBlend {blendOption=Linear}
    boundBone {boneName=dyn_shoe_strings_bot_r_r}
    componentAppliedLinearAccClamp=<5000 5000 5000>
    componentLinearAccScale=<0.6 0.6 0.6>
    componentLinearVelScale=<0.6 0.6 0.6>
    gravityOverride {z=-980}
    physicsBodyDefinitions [{
      sphereCollisionRadius=0
      boundBone {boneName=dyn_shoe_strings_bot_r_r}
      boxExtents=<14.28 6.86 7.14>
      constraintSetup {angularLimitsMax=<7 5 0>, angularLimitsMin=<-10 -12 -70>, angularTarget=<0 0 0>}
      localJointOffset=<3.724 0 0>
     }]
    relativeSpaceBone {boneName=foot_l}
   }
   n9 = Anim Dynamics - Bone: dyn_shoe_strings_top_l_r!(ComponentPose=@n8, Alpha=1, GravityScale:float=1, GravityOverride:Vector="0,0,-980", ExternalForce="0,0,0") -> (Pose) {
    angularDampingOverride=0.9999
    angularSpringConstant=100
    bAngularSpring=true
    bGravityOverrideInSimSpace=true
    bOverrideAngularDamping=true
    simulationSpace=RootRelative
    alphaBoolBlend {blendOption=Linear}
    boundBone {boneName=dyn_shoe_strings_top_l_r}
    componentAppliedLinearAccClamp=<5000 5000 5000>
    componentLinearAccScale=<0.6 0.6 0.6>
    componentLinearVelScale=<0.6 0.6 0.6>
    gravityOverride {z=-980}
    physicsBodyDefinitions [{
      sphereCollisionRadius=0
      boundBone {boneName=dyn_shoe_strings_top_l_r}
      boxExtents=<14 7 7>
      constraintSetup {angularLimitsMax=<7 5 0>, angularLimitsMin=<-10 -12 -70>, angularTarget=<0 0 0>}
      localJointOffset=<3.92 0 0>
     }]
    relativeSpaceBone {boneName=foot_l}
   }
   n10 = Anim Dynamics - Bone: dyn_shoe_strings_top_r_r!(ComponentPose=@n9, Alpha=1, GravityScale:float=1, GravityOverride:Vector="0,0,-980", ExternalForce="0,0,0") -> (Pose) {
    angularDampingOverride=0.9999
    angularSpringConstant=100
    bAngularSpring=true
    bGravityOverrideInSimSpace=true
    bOverrideAngularDamping=true
    simulationSpace=RootRelative
    alphaBoolBlend {blendOption=Linear}
    boundBone {boneName=dyn_shoe_strings_top_r_r}
    componentAppliedLinearAccClamp=<5000 5000 5000>
    componentLinearAccScale=<0.6 0.6 0.6>
    componentLinearVelScale=<0.6 0.6 0.6>
    gravityOverride {z=-980}
    physicsBodyDefinitions [{
      sphereCollisionRadius=0
      boundBone {boneName=dyn_shoe_strings_top_r_r}
      boxExtents=<14.28 6.86 6.72>
      constraintSetup {angularLimitsMax=<7 5 0>, angularLimitsMin=<-10 -12 -70>, angularTarget=<0 0 0>}
      localJointOffset=<4.116 0 0>
     }]
    relativeSpaceBone {boneName=foot_l}
   }
   n11 = Anim Dynamics - Bone: dyn_belt_bag_front_01!(ComponentPose=@n10, Alpha=1, GravityScale:float=1, GravityOverride:Vector="0,0,-980", ExternalForce="0,0,0") -> (Pose) {
    angularDampingOverride=0.9999
    angularSpringConstant=200
    bAngularSpring=true
    bGravityOverrideInSimSpace=true
    bOverrideAngularDamping=true
    simulationSpace=RootRelative
    alphaBoolBlend {blendOption=Linear}
    boundBone {boneName=dyn_belt_bag_front_01}
    componentAppliedLinearAccClamp=<5000 5000 5000>
    componentLinearAccScale=<0.6 0.6 0.6>
    componentLinearVelScale=<0.6 0.6 0.6>
    gravityOverride {z=-980}
    physicsBodyDefinitions [{
      sphereCollisionRadius=0
      boundBone {boneName=dyn_belt_bag_front_01}
      boxExtents=<30 15 15>
      constraintSetup {angularLimitsMax=<0.5 0.5 0>, angularLimitsMin=<-0.5 -0.5 -35>, angularTarget=<1 1 1>}
      localJointOffset=<2 0 0>
     }]
    relativeSpaceBone {boneName=foot_l}
   }
   n12 = Anim Dynamics - Bone: dyn_belt_bag_back_01!(ComponentPose=@n11, Alpha=1, GravityScale:float=1, GravityOverride:Vector="0,0,-980", ExternalForce="0,0,0") -> (Pose) {
    angularDampingOverride=0.9999
    angularSpringConstant=200
    bAngularSpring=true
    bGravityOverrideInSimSpace=true
    bOverrideAngularDamping=true
    simulationSpace=RootRelative
    alphaBoolBlend {blendOption=Linear}
    boundBone {boneName=dyn_belt_bag_back_01}
    componentAppliedLinearAccClamp=<5000 5000 5000>
    componentLinearAccScale=<0.6 0.6 0.6>
    componentLinearVelScale=<0.6 0.6 0.6>
    gravityOverride {z=-980}
    physicsBodyDefinitions [{
      sphereCollisionRadius=0
      boundBone {boneName=dyn_belt_bag_back_01}
      boxExtents=<28 14 16>
      constraintSetup {angularLimitsMax=<0.5 0.5 0>, angularLimitsMin=<-0.5 -0.5 -34>, angularTarget=<1 1 1>}
      localJointOffset=<1.8 0 0>
     }]
    relativeSpaceBone {boneName=foot_l}
   }
   n13 = RigidBody!(ComponentPose=@n12, Alpha, OverrideWorldGravity:Vector="0,0,-980", ExternalForce) -> (Pose) {
    bClampLinearTranslationLimitToRefPose=true
    bOverrideWorldGravity=true
    overridePhysicsAsset=&/M_MED_CashewRoll_Physics_Skarf
    simulationSpace=BaseBoneSpace
    alphaBoolBlend {blendOption=Linear}
    baseBoneRef {boneName=spine_04}
    componentLinearAccScale=<0.9 0.7 0.7>
    componentLinearVelScale=<0.9 0.7 0.7>
    overrideWorldGravity {z=-980}
   }
   n14 = Get IsFrontEnd() -> (IsFrontEnd:double)
   n15 = "Transform (Modify) Bone - Bone: dyn_skarf_bow_01_l"(ComponentPose=@n13, Alpha=@n14, Translation="0,0,0", Rotation="5,3,0", Scale="1,1,1") {rotationMode=BMM_Additive, alphaBoolBlend {blendOption=Linear}, boneToModify {boneName=dyn_skarf_bow_01_l}, rotation {pitch=5, yaw=3}}
   n16 = "Transform (Modify) Bone - Bone: dyn_skarf_bow_02_l"(ComponentPose=@n15, Alpha=@n14, Translation="0,0,0", Rotation="5,0,0", Scale="1,1,1") {rotationMode=BMM_Additive, alphaBoolBlend {blendOption=Linear}, boneToModify {boneName=dyn_skarf_bow_02_l}, rotation {pitch=5}}
   n17 = "Transform (Modify) Bone - Bone: dyn_skarf_bow_01_r"(ComponentPose=@n16, Alpha=@n14, Translation="0,0,0", Rotation="-7,-3,0", Scale="1,1,1") {rotationMode=BMM_Additive, alphaBoolBlend {blendOption=Linear}, boneToModify {boneName=dyn_skarf_bow_01_r}, rotation {pitch=-7, yaw=-3}}
   n18 = "Transform (Modify) Bone - Bone: dyn_skarf_bow_02_r"(ComponentPose=@n17, Alpha=@n14, Translation="0,0,0", Rotation="-2,0,0", Scale="1,1,1") {rotationMode=BMM_Additive, alphaBoolBlend {blendOption=Linear}, boneToModify {boneName=dyn_skarf_bow_02_r}, rotation {pitch=-2}}
   Save cached pose 'FECustomIdleCache'(Pose=Component To Local(ComponentPose=@n18)) {cacheName=FECustomIdleCache}
  ]}
 EventGraph {nodes [
   n1 = Event BlueprintUpdateAnimation()
   > n2 = Sequence!(execute) -> (then_0, then_1, then_2:Exec, then_3:Exec, then_4:Exec)
   > n3 = Gravity Override State Function DEPRECATED(CharacterPartAnimInstance=Self-Reference(), GameTimeInSeconds=GetGameTimeInSeconds(), GravityOverrideParams=Get Gravity Override Params())
   > Set GravityOverride(GravityOverride=@n3)
   Set WPO Globals Function(
    execute=@n2.then_2
    Material=Get dynamic_mat_instance() -> (dynamic_mat_instance:Material Instance Dynamic Object Reference)
    Overlay Material=Get dynamic_overlay_mat() -> (dynamic_overlay_mat:Material Instance Dynamic Object Reference)
    Skeletal Mesh=Get mesh() -> (mesh:Skeletal Mesh Component Budgeted Object Reference)
    "Local Noise Transform  Location"="0, 0, 0"
    Local Noise Offset="0, 0, 0"
   )
   > Set WPO Regional Function!(
    execute
    Material:Material Instance Dynamic Object Reference=Get dynamic_mat_instance()
    Overlay Material:Material Instance Dynamic Object Reference=Get dynamic_overlay_mat()
    Skeletal Mesh:Skeletal Mesh Component Object Reference=Get mesh()
    Velocity Transform:Name=calf_r
    Velocity Interp Speed:double=Get Velocity Interp Speed_reg1() -> (Velocity Interp Speed_reg1:double)
    Collision Transform:Name=calf_r
    Collision Transform Axis:EAxis Enum
    Invert Collision Transform Axis:bool=false
    Delta Time X:double=@n1.DeltaTimeX
    Previous Velocity Transform Location:"Vector (by ref)"=Get Previous Velocity Transform Location_reg1() -> (Previous Velocity Transform Location_reg1:Vector)
    Previous Velocity Transform Velocity:"Vector (by ref)"=Get Previous Velocity Transform Velocity_reg1() -> (Previous Velocity Transform Velocity_reg1:Vector)
    Region:int=1
   ) -> (then)
   > Set WPO Regional Function(
    Material=Get dynamic_mat_instance()
    Overlay Material=Get dynamic_overlay_mat()
    Skeletal Mesh=Get mesh()
    Velocity Transform=calf_l
    Velocity Interp Speed=Get Velocity Interp Speed_reg2() -> (Velocity Interp Speed_reg2:double)
    Collision Transform=calf_l
    Invert Collision Transform Axis=true
    Delta Time X=@n1.DeltaTimeX
    Previous Velocity Transform Location=Get Previous Velocity Transform Location_reg2() -> (Previous Velocity Transform Location_reg2:Vector)
    Previous Velocity Transform Velocity=Get Previous Velocity Transform Velocity_reg2() -> (Previous Velocity Transform Velocity_reg2:Vector)
    Region=2
   )
   n4 = Get dynamic_mat_instance()
   n5 = Get dynamic_overlay_mat()
   > SetScalarParameterValue(self=[@n4, @n5], ParameterName=ClothSim_WindStrength, Value=Get wind_strength() -> (wind_strength:float))
   > SetScalarParameterValue(
    self=[@n4, @n5]
    ParameterName=ClothSim_TurbulenceNoiseTileDivisor
    Value=Get TurbulenceNoiseTileDivisor() -> (TurbulenceNoiseTileDivisor:float)
   )
   n6 = Branch(execute=@n2.then_3, Condition=Get bIsInFrontEnd())
   > Set CustomIdleIsPlayingFloat(CustomIdleIsPlayingFloat:float=GetCurveValue(self=GetOwnerFrontEndAnimInstance(), CurveName=Frontend_Idle_Bypass)) -> (Output_Get:float)
   Set CustomIdleIsPlayingFloat(execute=@n6.else, CustomIdleIsPlayingFloat=0)
   n7 = Branch(execute=@n2.then_4, Condition=Get bIsInFrontEnd())
   > Set IsFrontEnd(IsFrontEnd:double=1) -> (Output_Get:double)
   Set IsFrontEnd(execute=@n7.else, IsFrontEnd=0)
   n8 = Branch(execute=@n2.then_1, Condition=Get bIsSkyDiving())
   > Set Velocity Interp Speed_reg1(Velocity Interp Speed_reg1:double=100) -> (Output_Get:double)
   > Set Velocity Interp Speed_reg2(Velocity Interp Speed_reg2:double=50) -> (Output_Get:double)
   > Set wind_strength(wind_strength:float=50) -> (Output_Get:float)
   > Set TurbulenceNoiseTileDivisor(TurbulenceNoiseTileDivisor:float=100) -> (Output_Get:float)
   Set Velocity Interp Speed_reg1(execute=@n8.else, Velocity Interp Speed_reg1=0)
   > Set Velocity Interp Speed_reg2(Velocity Interp Speed_reg2=0)
   > Set wind_strength(wind_strength=0)
   > Set TurbulenceNoiseTileDivisor(TurbulenceNoiseTileDivisor=0)
   n9 = Event OnCharacterCustomizationLoaded()
   > n10 = Set mesh(mesh:Skeletal Mesh Component Budgeted Object Reference=Get Array Item with EFort Custom Part Type(Array=@n9.CustomMeshParts)) -> (Output_Get:Skeletal Mesh Component Budgeted Object Reference)
   > n11 = Get Dynamic Material Instance(SkeletalMeshComponent=@n10)
   > Set dynamic_mat_instance(dynamic_mat_instance:Material Instance Dynamic Object Reference=@n11) -> (Output_Get:Material Instance Dynamic Object Reference)
   > n12 = GetOrCreateDynamicOverlayMaterial(MeshComponent=Get mesh(), MaterialName=None)
   > Set dynamic_overlay_mat(dynamic_overlay_mat:Material Instance Dynamic Object Reference=@n12) -> (Output_Get:Material Instance Dynamic Object Reference)
  ]}
]
variables [
 CustomIdleIsPlayingFloat {type=number, default=1}
 dynamic_mat_instance {type=MaterialInstanceDynamic}
 dynamic_overlay_mat {type=MaterialInstanceDynamic}
 Gravity Override Params {type=GravityOverrideParamsStruct}
 GravityOverride {type=Vector}
 IsFrontEnd {type=number}
 mesh {type=SkeletalMeshComponentBudgeted}
 Previous Velocity Transform Location_reg1 {type=Vector}
 Previous Velocity Transform Location_reg2 {type=Vector}
 Previous Velocity Transform Velocity_reg1 {type=Vector}
 Previous Velocity Transform Velocity_reg2 {type=Vector}
 TurbulenceNoiseTileDivisor {type=number}
 Velocity Interp Speed_reg1 {type=number, default=1}
 Velocity Interp Speed_reg2 {type=number, default=1}
 wind_strength {type=number}
]
