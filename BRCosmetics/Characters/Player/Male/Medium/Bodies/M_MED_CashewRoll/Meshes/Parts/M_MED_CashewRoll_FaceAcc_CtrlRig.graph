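# Control Rig for the CashewRoll character's face accessory that adjusts the back hair chain based on head pose.
# A spherical pose reader measures the head bone against spine_05, with debug drawing left on, and outputs a falloff weight.
# That weight blends a fixed offset and rotation onto Dyn_Master_HairBack, set relative to its own child Dyn_HairBack1 and propagated to children.
# The other four Dyn hair chains are untouched.
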ControlRigBlueprint /BRCosmetics/Characters/Player/Male/Medium/Bodies/M_MED_CashewRoll/Meshes/Parts/M_MED_CashewRoll_FaceAcc_CtrlRig
graphs [
 RigVMFunctionLibrary {links [], nodes []}
 RigVMModel {nodes [
   RigUnit_BeginExecution = &RigUnit_BeginExecution()
   n1 = &RigUnit_SphericalPoseReader(
    ActiveRegionScaleFactors=(PositiveWidth=1.7,NegativeWidth=2,PositiveHeight=2,NegativeHeight=2)
    ActiveRegionSize=0.1
    Debug=(bDrawDebug=true,bDraw2D=False,bDrawLocalAxes=False,DebugScale=25,DebugSegments=20,DebugThickness=0.25)
    DriverAxis=(X=1,Y=0,Z=0)
    DriverItem=(Type=Bone,Name="head")
    FalloffRegionScaleFactors=(PositiveWidth=1.3,NegativeWidth=1,PositiveHeight=2,NegativeHeight=2)
    FalloffSize=0.3
    FlipHeightScaling=False
    FlipWidthScaling=False
    OptionalParentItem=(Type=Bone,Name="spine_05")
    RotationOffset=(X=-90,Y=0,Z=25)
    ExecutePin=@RigUnit_BeginExecution.ExecutePin
   )
   > /Script/ControlRig.RigUnit_SetRelativeTransformForItem(
    Child=(Type=Bone,Name="Dyn_Master_HairBack")
    Parent=(Type=Bone,Name="Dyn_HairBack1")
    Value=(Rotation=(X=0,Y=0,Z=0.1736,W=0.9848),Translation=(X=-8,Y=0,Z=0),Scale3D=(X=1,Y=1,Z=1))
    bParentInitial=false
    bPropagateToChildren=true
    Weight=@n1.OutputParam
   ) {
    is=struct RigVMUnitNode
    template="Set Relative Transform::Execute(in Child,in Parent,in bParentInitial,in Value,in Weight,in bPropagateToChildren)"
   }
  ]}
]
hierarchy [
 root {loc=<0 0 0>, rot=<0 0 0 1>}
 pelvis {parent=root, loc=<0 -0.7071 93.1191>, rot=<0.0127 0.707 -0.0127 -0.707>}
 spine_01 {parent=pelvis, loc=<6.7494 -0.6388 0>, rot=<0 0 0.034 -0.9994>}
 spine_02 {parent=spine_01, loc=<7.2518 -0.035 0>, rot=<0 0 0.0238 -0.9997>}
 spine_03 {parent=spine_02, loc=<6.0104 0 0>, rot=<0 0 0.0149 -0.9999>}
 spine_04 {parent=spine_03, loc=<8.1024 0 0>, rot=<0 0 -0.0132 -0.9999>}
 spine_05 {parent=spine_04, loc=<10.0146 0 0>, rot=<0 0 -0.0015 -1>}
 neck_01 {parent=spine_05, loc=<11.3138 0.1529 0>, rot=<0 0 -0.2003 -0.9797>}
 neck_02 {parent=neck_01, loc=<4.5627 0 0>, rot=<0 0 0.0504 -0.9987>}
 head {parent=neck_02, loc=<5.3914 0 0>, rot=<0 0 0.115 -0.9934>}
 Dyn_Master_Hair {parent=head, loc=<7.4064 0 0>, rot=<0 0 0 -1>}
 Dyn_Master_HairLeft {parent=Dyn_Master_Hair, loc=<0.585 1.267 -10.174>, rot=<0.0115 -0.707 -0.707 0.0115>}
 Dyn_HairLeft1 {parent=Dyn_Master_HairLeft, loc=<0.362 0 0>, rot=<0 0 0 1>}
 Dyn_HairLeft2 {parent=Dyn_HairLeft1, loc=<4.7624 0 0>, rot=<0 0 0 1>}
 Dyn_HairLeft3 {parent=Dyn_HairLeft2, loc=<5.102 0 0>, rot=<0 0 0 1>}
 Dyn_Master_HairRight {parent=Dyn_Master_Hair, loc=<0.5855 1.2563 10.1735>, rot=<0.0115 -0.707 0.707 -0.0115>}
 Dyn_HairRight1 {parent=Dyn_Master_HairRight, loc=<0.3621 0 0>, rot=<0 0 0 1>}
 Dyn_HairRight2 {parent=Dyn_HairRight1, loc=<4.7624 0 0>, rot=<0 0 0 1>}
 Dyn_HairRight3 {parent=Dyn_HairRight2, loc=<5.1025 0 0>, rot=<0 0 0 1>}
 Dyn_Master_HairBack {parent=Dyn_Master_Hair, loc=<0.9027 -9.922 0>, rot=<0.0163 -0.9999 0 0>}
 Dyn_HairBack1 {parent=Dyn_Master_HairBack, loc=<0.362 0 0>, rot=<0 0 0 1>}
 Dyn_HairBack2 {parent=Dyn_HairBack1, loc=<4.7624 0 0>, rot=<0 0 0 1>}
 Dyn_HairBack3 {parent=Dyn_HairBack2, loc=<5.102 0 0>, rot=<0 0 0 1>}
 Dyn_Master_HairLeftFront {parent=Dyn_Master_Hair, loc=<0.5151 9.7274 -8.677>, rot=<-0.0084 0.515 0.8571 -0.014>}
 Dyn_HairLeftFront1 {parent=Dyn_Master_HairLeftFront, loc=<0.362 0 0>, rot=<0 0 0 1>}
 Dyn_HairLeftFront2 {parent=Dyn_HairLeftFront1, loc=<4.7624 0 0>, rot=<0 0 0 1>}
 Dyn_HairLeftFront3 {parent=Dyn_HairLeftFront2, loc=<5.102 0 0>, rot=<0 0 0 1>}
 Dyn_Master_HairRightFront {parent=Dyn_Master_Hair, loc=<0.515 9.727 8.677>, rot=<0.0084 -0.515 0.8571 -0.014>}
 Dyn_HairRightFront1 {parent=Dyn_Master_HairRightFront, loc=<0.3621 0 0>, rot=<0 0 0 1>}
 Dyn_HairRightFront2 {parent=Dyn_HairRightFront1, loc=<4.7624 0 0>, rot=<0 0 0 1>}
 Dyn_HairRightFront3 {parent=Dyn_HairRightFront2, loc=<5.1025 0 0>, rot=<0 0 0 1>}
]
