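# Physics asset for the M_MED_CashewRoll face accessory mesh, defining physics bodies for dynamic hair on this male player cosmetic.
# Five three-capsule chains (back, left, right, left-front, right-front) each hang from a small master body constrained to a shared Dyn_Master_Hair root.
# Master-to-root joints swing and twist freely, while chain joints have tight angular limits.
# Back hair is stiffest, front strands are loosest, and side and front limits generally widen toward the tips.
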PhysicsAsset /BRCosmetics/Characters/Player/Male/Medium/Bodies/M_MED_CashewRoll/Meshes/Parts/M_MED_CashewRoll_FaceAcc_Physics
bodies {
 Dyn_HairBack1 [{shapeType=Capsule, length=3, radius=5, shapeName=Dyn_HairBack1_capsule, center=<4.6723 0 0>, rotation=<rot 90 0 0>}]
 Dyn_HairBack2 [{shapeType=Capsule, length=3, radius=5, shapeName=Dyn_HairBack2_capsule, center=<4.6723 0 0>, rotation=<rot 90 0 0>}]
 Dyn_HairBack3 [{shapeType=Capsule, length=1, radius=5, shapeName=Dyn_HairBack3_capsule, center=<3.3323 0 0>, rotation=<rot 90 0 0>}]
 Dyn_HairLeft1 [{shapeType=Capsule, length=3, radius=5, shapeName=Dyn_HairLeft1_capsule, center=<4.6723 0 0>, rotation=<rot 90 0 0>}]
 Dyn_HairLeft2 [{shapeType=Capsule, length=3, radius=5, shapeName=Dyn_HairLeft2_capsule, center=<4.6723 0 0>, rotation=<rot 90 0 0>}]
 Dyn_HairLeft3 [{shapeType=Capsule, length=1, radius=5, shapeName=Dyn_HairLeft3_capsule, center=<3.3323 0 0>, rotation=<rot 90 0 0>}]
 Dyn_HairLeftFront1 [{shapeType=Capsule, length=3, radius=5, shapeName=Dyn_HairLeft1_capsule, center=<4.6723 0 0>, rotation=<rot 90 0 0>}]
 Dyn_HairLeftFront2 [{shapeType=Capsule, length=3, radius=5, shapeName=Dyn_HairLeft2_capsule, center=<4.6723 0 0>, rotation=<rot 90 0 0>}]
 Dyn_HairLeftFront3 [{shapeType=Capsule, length=1, radius=5, shapeName=Dyn_HairLeft3_capsule, center=<3.3323 0 0>, rotation=<rot 90 0 0>}]
 Dyn_HairRight1 [{shapeType=Capsule, length=3, radius=5, shapeName=Dyn_HairRight1_capsule, center=<4.6723 0 0>, rotation=<rot 90 0 0>}]
 Dyn_HairRight2 [{shapeType=Capsule, length=3, radius=5, shapeName=Dyn_HairRight2_capsule, center=<4.6723 0 0>, rotation=<rot 90 0 0>}]
 Dyn_HairRight3 [{shapeType=Capsule, length=1, radius=5, shapeName=Dyn_HairRight3_capsule, center=<3.3323 0 0>, rotation=<rot 90 0 0>}]
 Dyn_HairRightFront1 [{shapeType=Capsule, length=3, radius=5, shapeName=Dyn_HairRight1_capsule, center=<4.6723 0 0>, rotation=<rot 90 0 0>}]
 Dyn_HairRightFront2 [{shapeType=Capsule, length=3, radius=5, shapeName=Dyn_HairRight2_capsule, center=<4.6723 0 0>, rotation=<rot 90 0 0>}]
 Dyn_HairRightFront3 [{shapeType=Capsule, length=1, radius=5, shapeName=Dyn_HairRight3_capsule, center=<3.3323 0 0>, rotation=<rot 90 0 0>}]
 Dyn_Master_Hair [{shapeType=Capsule, length=0.5, radius=0.5, shapeName=Dyn_Master_Hair_capsule}]
 Dyn_Master_HairBack [{shapeType=Capsule, length=0.505, radius=0.505, shapeName=Dyn_Master_HairBack_capsule, rotation=<rot 0 -90 0>}]
 Dyn_Master_HairLeft [{shapeType=Capsule, length=0.5, radius=0.5, shapeName=Dyn_Master_HairLeft_capsule}]
 Dyn_Master_HairLeftFront [{shapeType=Capsule, length=0.505, radius=0.505, shapeName=Dyn_Master_HairLeftFront_capsule, rotation=<rot 0 -90 0>}]
 Dyn_Master_HairRight [{shapeType=Capsule, length=0.5, radius=0.5, shapeName=Dyn_Master_HairRight_capsule}]
 Dyn_Master_HairRightFront [{shapeType=Capsule, length=0.505, radius=0.505, shapeName=Dyn_Master_HairRightFront_capsule, rotation=<rot 0 -90 0>}]
}
constraints [
 {
  bone1Name=Dyn_Master_HairLeft
  bone2Name=Dyn_Master_Hair
  swing1LimitDegrees=45
  swing1Motion=Free
  swing2LimitDegrees=45
  swing2Motion=Free
  twistLimitDegrees=45
  twistMotion=Free
 }
 {
  bone1Name=Dyn_Master_HairRight
  bone2Name=Dyn_Master_Hair
  swing1LimitDegrees=45
  swing1Motion=Free
  swing2LimitDegrees=45
  swing2Motion=Free
  twistLimitDegrees=45
  twistMotion=Free
 }
 {bone1Name=Dyn_HairBack1, bone2Name=Dyn_Master_HairBack, swing1LimitDegrees=4, swing2LimitDegrees=3, twistLimitDegrees=1}
 {bone1Name=Dyn_HairBack2, bone2Name=Dyn_HairBack1, swing1LimitDegrees=7, swing2LimitDegrees=3, twistLimitDegrees=1}
 {bone1Name=Dyn_HairBack3, bone2Name=Dyn_HairBack2, swing1LimitDegrees=5, swing2LimitDegrees=3, twistLimitDegrees=1}
 {bone1Name=Dyn_HairLeft1, bone2Name=Dyn_Master_HairLeft, swing1LimitDegrees=5, swing2LimitDegrees=10, twistLimitDegrees=1}
 {bone1Name=Dyn_HairLeft2, bone2Name=Dyn_HairLeft1, swing1LimitDegrees=8, swing2LimitDegrees=10, twistLimitDegrees=1}
 {bone1Name=Dyn_HairLeft3, bone2Name=Dyn_HairLeft2, swing1LimitDegrees=12, swing2LimitDegrees=15, twistLimitDegrees=1}
 {bone1Name=Dyn_HairRight1, bone2Name=Dyn_Master_HairRight, swing1LimitDegrees=5, swing2LimitDegrees=10, twistLimitDegrees=1}
 {bone1Name=Dyn_HairRight2, bone2Name=Dyn_HairRight1, swing1LimitDegrees=8, swing2LimitDegrees=10, twistLimitDegrees=1}
 {bone1Name=Dyn_HairRight3, bone2Name=Dyn_HairRight2, swing1LimitDegrees=12, swing2LimitDegrees=15, twistLimitDegrees=1}
 {
  bone1Name=Dyn_Master_HairBack
  bone2Name=Dyn_Master_Hair
  swing1LimitDegrees=45
  swing1Motion=Free
  swing2LimitDegrees=45
  swing2Motion=Free
  twistLimitDegrees=45
  twistMotion=Free
 }
 {
  bone1Name=Dyn_Master_HairLeftFront
  bone2Name=Dyn_Master_Hair
  swing1LimitDegrees=45
  swing1Motion=Free
  swing2LimitDegrees=45
  swing2Motion=Free
  twistLimitDegrees=45
  twistMotion=Free
 }
 {bone1Name=Dyn_HairLeftFront1, bone2Name=Dyn_Master_HairLeftFront, swing1LimitDegrees=10, swing2LimitDegrees=15, twistLimitDegrees=8}
 {bone1Name=Dyn_HairLeftFront2, bone2Name=Dyn_HairLeftFront1, swing1LimitDegrees=15, swing2LimitDegrees=15, twistLimitDegrees=8}
 {bone1Name=Dyn_HairLeftFront3, bone2Name=Dyn_HairLeftFront2, swing1LimitDegrees=15, swing2LimitDegrees=20, twistLimitDegrees=8}
 {
  bone1Name=Dyn_Master_HairRightFront
  bone2Name=Dyn_Master_Hair
  swing1LimitDegrees=45
  swing1Motion=Free
  swing2LimitDegrees=45
  swing2Motion=Free
  twistLimitDegrees=45
  twistMotion=Free
 }
 {bone1Name=Dyn_HairRightFront1, bone2Name=Dyn_Master_HairRightFront, swing1LimitDegrees=10, swing2LimitDegrees=15, twistLimitDegrees=8}
 {bone1Name=Dyn_HairRightFront2, bone2Name=Dyn_HairRightFront1, swing1LimitDegrees=15, swing2LimitDegrees=15, twistLimitDegrees=8}
 {bone1Name=Dyn_HairRightFront3, bone2Name=Dyn_HairRightFront2, swing1LimitDegrees=15, swing2LimitDegrees=20, twistLimitDegrees=8}
]
